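# Material function, exposed to the library, that returns sunlight Direction and Color from the FortniteMaterialParameters collection.
# Direction lerps from the sky atmosphere light direction to the RGB of SunlightDirectionalVector, using that vector's alpha as the blend to flag whether it is being used.
# Color passes through the RGB of SunLightColor.
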MaterialFunction /Game/Global/MaterialFunctions/FNCollectionParameters/MF_SunParameters_FNC
graph {
 properties {
  bExposeToLibrary=true
  description="Returns sunlight direction and color from FortniteMaterialParameters, values "
  libraryCategoriesText [FN Collection Parameters]
 }
 nodes [
  n1 = CollectionParameter() {
   collection=&FortniteMaterialParameters
   parameterName=SunlightDirectionalVector
  }
  FunctionOutput(None=LinearInterpolate(A=SkyAtmosphereLightDirection(), B=ComponentMask(None=@n1) {b=true, g=true, r=true}, Alpha=ComponentMask(None=@n1) {a=true}) {desc=Check if Sunlight Direction is being used}) {outputName=Direction}
  FunctionOutput(None=ComponentMask(None=CollectionParameter() {collection=&FortniteMaterialParameters, parameterName=SunLightColor}) {b=true, g=true, r=true}) {outputName=Color, sortPriority=10}
 ]
}
